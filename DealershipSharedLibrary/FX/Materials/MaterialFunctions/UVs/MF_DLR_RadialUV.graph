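# Dealership library material function that converts input UVs into tiled, panning radial UVs centered at 0.5.
# U is square-rooted center distance scaled by Exp X and raised to Twist; V adds normalized atan2 angle to U.
# Both atan2 inputs use the G component, so the angle only steps between halves rather than sweeping around.
# It also outputs DDX and DDY of the unmodified UVs.
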MaterialFunction /DealershipSharedLibrary/FX/Materials/MaterialFunctions/UVs/MF_DLR_RadialUV
graph {
 properties {bExposeToLibrary=true, userExposedCaption=DLR Radial UVs, libraryCategoriesText [Dealership]}
 nodes [
  n1 = FunctionInput(Preview=TextureCoordinate()) {inputName=UV In, inputType=FunctionInput_Vector2}
  n2 = Subtract(A=@n1) {constB=0.5}
  n3 = MaterialFunctionCall(Float2=@n2) {materialFunction=&BreakOutFloat2Components}
  n4 = Add(A=Arctangent2Fast(Y=@n3.G, X=@n3.G), B=MaterialFunctionCall() {materialFunction=&Pi}.Result)
  n5 = MaterialFunctionCall(Multiplier=Constant() {r=2}) {materialFunction=&Pi}
  n6 = MaterialFunctionCall(Vector 2=@n2) {materialFunction=/Engine/Functions/Engine_MaterialFunctions02/Utility/VectorLength}
  n7 = Power(
   Base=Multiply(
    A=SquareRoot(None=@n6.V2 Length)
    B=FunctionInput() {inputName=Exp X, inputType=FunctionInput_Scalar, sortPriority=6, previewValue=<1 0 0 1>}
   )
   Exp=FunctionInput() {inputName=Twist, inputType=FunctionInput_Scalar, sortPriority=5, previewValue=<1 0 0 1>}
  )
  n8 = AppendVector(A=@n7, B=Add(A=@n7, B=Divide(A=@n4, B=@n5.Result)))
  n9 = AppendVector(
   A=FunctionInput() {inputName=Tile X, inputType=FunctionInput_Scalar, sortPriority=1, previewValue=<1 0 0 1>}
   B=FunctionInput() {inputName=Tile Y, inputType=FunctionInput_Scalar, sortPriority=2, previewValue=<1 0 0 1>}
  )
  n10 = FunctionInput() {inputName=Panning Speed X, inputType=FunctionInput_Scalar, sortPriority=3, previewValue=<-0.25 0 0 1>}
  n11 = FunctionInput() {inputName=Panning Speed Y, inputType=FunctionInput_Scalar, sortPriority=4, previewValue=<0.25 0 0 1>}
  FunctionOutput(None=Panner(Coordinate=Multiply(A=@n8, B=@n9), Time=Time(), Speed=AppendVector(A=@n10, B=@n11)) {speedY=1}) {outputName=Input UVs}
  FunctionOutput(None=DDX(Value=@n1)) {outputName=DDX, sortPriority=1}
  FunctionOutput(None=DDY(Value=@n1)) {outputName=DDY, sortPriority=2}
 ]
}
